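# Placeholder Default Lit material, going by its _Temp name, for the professional kitchen counter prop in the Asteria environment.
# The graph holds only four scalar constants, with no textures, parameters or visible connections to material outputs.
# It is flagged for use with Nanite meshes.
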Material /Game/Environments/Asteria/Props/Kitchen_Pro/Kitchen_Pro_Counter/Materials/_Temp_Counter
graph {
 properties {bUsedWithNanite=true, pixelDepthOffsetMode=PDOM_Legacy, usedShadingModels=Default Lit}
 nodes [
  Constant() {r=0.3}
  Constant() {r=0.3}
  Constant() {r=0.6}
  Constant() {r=1}
 ]
}
